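# Niagara prop-spawn effect for the CozyMechanic emote that throws a short-lived shower of nuts-and-bolts meshes.
# Nuts_and_Bolts fires three staggered bursts, using spawn group to index arrays for minimum scale and launch velocity.
# A random VisibilityTag accompanies two mesh renderers, and particles tumble, fall under heavy gravity and shrink away.
# The Airglow, Smoke and Sparkles emitters show no modules or renderers here.
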NiagaraSystem /BRCosmetics/Animation/Game/MainPlayer/Emotes/CozyMechanic/FX/NS_Default_PropSpawn_LWE
dependencies {
 usedDataInterfaces [&NiagaraDataInterfaceArrayFloat, &NiagaraDataInterfaceArrayFloat3, &NiagaraDataInterfaceCurve]
 usedDynamicInputs []
 usedModules [&SystemState]
}
emitters {
 AirglowLWE {rendererClasses [], renderers []}
 Nuts_and_Bolts {
  emitterUpdateScript [
   &EmitterState()
   &SpawnBurst_Instantaneous(Spawn Count=25, Use Spawn Probability=false, UseLoopCountLimit=false)
   &SpawnBurst_Instantaneous(Spawn Count=12, Spawn Group=1, Spawn Time=0.05, Use Spawn Probability=false, UseLoopCountLimit=false) {module=SpawnBurst_Instantaneous001}
   &SpawnBurst_Instantaneous(Spawn Count=2, Spawn Group=2, Spawn Time=0.08, Use Spawn Probability=false, UseLoopCountLimit=false) {module=SpawnBurst_Instantaneous002}
  ]
  particleSpawnScript [
   &PartitionParticles(Number Of Partitions=3)
   &InitializeParticle(
    Color=<rgba 0.2865 0.2865 0.2865 1>
    Lifetime=&RandomRangeFloat(Minimum=1.3, Maximum=2.1)
    Mesh Scale Mode=::NewEnumerator3
    Mesh Uniform Scale=&RandomRangeFloat(
     Minimum=&SelectFloatFromArray(
      Direct Array Index=&SpawnGroup()
      Float Selection Array=DataInterface() {propertyValues {FloatData [0.5,0.8,1.5]}}
     )
     Maximum=2
    )
    UsePositionOffset=false
    Write Material Random=true
   )
   &ShapeLocation(MaskSpawnGroup=false, Override Seed=false, Sphere Radius=5, UseOwnerScale=true)
   &AddVelocity(
    Module.Use Falloff Distance=false
    Origin Offset=<0 0 -1>
    Override Seed=false
    Velocity Mode=::NewEnumerator1
    Velocity Speed=&RandomRangeFloat(Minimum=100, Maximum=200)
   )
   &AddVelocity(Velocity=&SelectVectorFromArray(
     Direct Array Index=&SpawnGroup()
     Vector Selection Array=DataInterface() {propertyValues {FloatData [<0 0 350>, <1 1 250>, <0 0 600>]}}
    )) {module=AddVelocity001}
   /SetVariables(Particles.VisibilityTag=&RandomRangeInt(Minimum=0, Maximum=1)) {isSetParameters=true}
   &InitialMeshOrientation(Use Rotation Vector=false)
  ]
  particleUpdateScript [
   &ParticleState()
   &ScaleMeshSize(Scale Factor=&Multiply_VectorByFloat(
     Vector=$Particles.Initial.Scale
     Float=&FloatFromCurve(FloatCurve=DataInterface() {propertyValues {Curve {keys [<at 0 1>,<at 1 0>]}}})
    ))
   &UpdateMeshOrientation(Rotation Rate=&RandomRangeFloat(Minimum=1, Maximum=5))
   &GravityForce(Gravity=<0 0 -1500>)
   &SolveForcesAndVelocity(Clamp Velocity=false, Limit Acceleration=false)
  ]
  rendererClasses [&NiagaraMeshRendererProperties]
  renderers [
   {rendererClass=&NiagaraMeshRendererProperties, rendererIndex=0}
   {rendererClass=&NiagaraMeshRendererProperties, rendererIndex=1}
  ]
 }
 SmokeLWE {rendererClasses [], renderers []}
 SparklesLWE {rendererClasses [], renderers []}
}
summary {
 systemName=NS_Default_PropSpawn_LWE
 emitters [
  {emitterName=SparklesLWE, rendererClasses []}
  {emitterName=AirglowLWE, rendererClasses []}
  {emitterName=SmokeLWE, rendererClasses []}
  {emitterName=Nuts_and_Bolts, rendererClasses [&NiagaraMeshRendererProperties]}
 ]
 userVariables [User.NormalizedNotifyProgress {type=NiagaraFloat, defaultValue=0, description=""}]
}
